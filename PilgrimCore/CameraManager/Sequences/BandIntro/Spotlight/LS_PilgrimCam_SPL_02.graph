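# Level Sequence for a PilgrimCore band intro spotlight camera shot, a 10-second cut at 30 fps.
# A near-zero-scale cube yaws quickly, then drifts slowly, apparently as a pivot for the attached cine camera, whose huge local offsets suit that scale.
# The camera zooms in slightly with fixed manual focus and wide aperture.
# A spotlight relocator actor sits at origin, and an animated player mannequin has spawning off by default.
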LevelSequence /PilgrimCore/CameraManager/Sequences/BandIntro/Spotlight/LS_PilgrimCam_SPL_02
bindings [
 Fort Player Mannequin {
  id=#1
  possessed_class=FortPlayerMannequin
  tracks [
   {
    type=MovieScene3DTransformTrack
    sections [{
      type=MovieScene3DTransformSection
      channels [
       Location.X {type=MovieSceneScriptingDoubleChannel, default=0}
       Location.Y {type=MovieSceneScriptingDoubleChannel, default=0}
       Location.Z {type=MovieSceneScriptingDoubleChannel, default=-90.7379}
       Rotation.X {type=MovieSceneScriptingDoubleChannel, default=0}
       Rotation.Y {type=MovieSceneScriptingDoubleChannel, default=0}
       Rotation.Z {type=MovieSceneScriptingDoubleChannel, default=-90}
       Scale.X {type=MovieSceneScriptingDoubleChannel, default=1}
       Scale.Y {type=MovieSceneScriptingDoubleChannel, default=1}
       Scale.Z {type=MovieSceneScriptingDoubleChannel, default=1}
      ]
     }]
   }
   {type=MovieSceneSkeletalAnimationTrack, sections [{type=MovieSceneSkeletalAnimationSection, end_frame=354, start_frame=0}]}
   {type=MovieSceneSpawnTrack, sections [{type=MovieSceneSpawnSection, channels [None {type=MovieSceneScriptingBoolChannel, default=false}]}]}
   {type=MovieScene3DAttachTrack, sections [{type=MovieScene3DAttachSection, end_frame=354, start_frame=0}]}
  ]
 }
 spotLight_relocator {
  id=#2
  possessed_class=Actor
  tracks [
   {type=MovieSceneSpawnTrack, sections [{type=MovieSceneSpawnSection, channels [None {type=MovieSceneScriptingBoolChannel, default=true}]}]}
   {
    type=MovieScene3DTransformTrack
    sections [{
      type=MovieScene3DTransformSection
      channels [
       Location.X {type=MovieSceneScriptingDoubleChannel, default=0}
       Location.Y {type=MovieSceneScriptingDoubleChannel, default=0}
       Location.Z {type=MovieSceneScriptingDoubleChannel, default=0}
       Rotation.X {type=MovieSceneScriptingDoubleChannel, default=0}
       Rotation.Y {type=MovieSceneScriptingDoubleChannel, default=0}
       Rotation.Z {type=MovieSceneScriptingDoubleChannel, default=0}
       Scale.X {type=MovieSceneScriptingDoubleChannel, default=1}
       Scale.Y {type=MovieSceneScriptingDoubleChannel, default=1}
       Scale.Z {type=MovieSceneScriptingDoubleChannel, default=1}
      ]
     }]
   }
  ]
 }
 FortPlayerCube {
  id=#3
  possessed_class=FortStaticMeshActor
  tracks [
   {
    type=MovieScene3DTransformTrack
    sections [{
      type=MovieScene3DTransformSection
      channels [
       Location.X {type=MovieSceneScriptingDoubleChannel, default=0, const [0,3]}
       Location.Y {type=MovieSceneScriptingDoubleChannel, default=0, const [0,3]}
       Location.Z {type=MovieSceneScriptingDoubleChannel, default=0, const [0,3]}
       Rotation.X {type=MovieSceneScriptingDoubleChannel, default=8, thinned_from=3, keys [[0,9.1964],[300,9.2068]]}
       Rotation.Y {type=MovieSceneScriptingDoubleChannel, default=0, keys [[0,18.0872],[18,18.2387],[300,18.2822]]}
       Rotation.Z {type=MovieSceneScriptingDoubleChannel, default=273.5992, keys [[0,85.8918],[18,0.7803],[300,-47.9178]]}
       Scale.X {type=MovieSceneScriptingDoubleChannel, default=0.0001, thinned_from=3, keys [[0,0.0001],[300,0.0001]]}
       Scale.Y {type=MovieSceneScriptingDoubleChannel, default=0.0001, thinned_from=3, keys [[0,0.0001],[300,0.0001]]}
       Scale.Z {type=MovieSceneScriptingDoubleChannel, default=0.0001, thinned_from=3, keys [[0,0.0001],[300,0.0001]]}
      ]
     }]
   }
   {type=MovieSceneSpawnTrack, sections [{type=MovieSceneSpawnSection, channels [None {type=MovieSceneScriptingBoolChannel, default=true}]}]}
   {type=MovieScene3DAttachTrack, sections [{type=MovieScene3DAttachSection, end_frame=300, start_frame=0}]}
  ]
 }
 CameraComponent {
  id=#4
  parent_id=#5
  possessed_class=CineCameraComponent
  tracks [
   {
    type=MovieSceneFloatTrack
    display_name=Current Focal Length
    sections [{
      type=MovieSceneFloatSection
      channels [None {type=MovieSceneScriptingFloatChannel, default=12, keys [[0,15.2479],[15,15.6755],[299,15.9095]]}]
     }]
   }
   {
    type=MovieSceneFloatTrack
    display_name="Manual Focus Distance (Focus Settings)"
    sections [{type=MovieSceneFloatSection, channels [None {type=MovieSceneScriptingFloatChannel, default=233.0195}]}]
   }
   {
    type=MovieSceneFloatTrack
    display_name=Current Aperture
    sections [{type=MovieSceneFloatSection, channels [None {type=MovieSceneScriptingFloatChannel, default=0.5}]}]
   }
  ]
 }
 Cine Camera Actor {
  id=#5
  possessed_class=CineCameraActor
  tracks [
   {type=MovieSceneSpawnTrack, sections [{type=MovieSceneSpawnSection, channels [None {type=MovieSceneScriptingBoolChannel, default=true}]}]}
   {type=MovieSceneCameraShakeTrack, sections []}
   {
    type=MovieScene3DTransformTrack
    sections [{
      type=MovieScene3DTransformSection
      channels [
       Location.X {type=MovieSceneScriptingDoubleChannel, default=185.8875, thinned_from=5, keys [[-0.0013,162.2978],[0,2177939.6012],[300,2177939.6012]]}
       Location.Y {type=MovieSceneScriptingDoubleChannel, default=0, thinned_from=5, keys [[-0.0013,-183.39],[0,207911.6621],[300,207911.6621]]}
       Location.Z {type=MovieSceneScriptingDoubleChannel, default=0, thinned_from=5, keys [[-0.0013,-46.9675], [0, -1196325.1233], [300, -1196325.1233]]}
       Rotation.X {type=MovieSceneScriptingDoubleChannel, default=0, thinned_from=5, keys [[-0.0013,0],[0,10.5174],[299.9987,10.5174],[300,-1.1566]]}
       Rotation.Y {type=MovieSceneScriptingDoubleChannel, default=-3.1144, thinned_from=5, keys [[-0.0013,2.0856],[0,19.7034],[299.9987,19.7034],[300,38.6071]]}
       Rotation.Z {
        type=MovieSceneScriptingDoubleChannel
        default=-180
        thinned_from=5
        keys [[-0.0013,133.4],[0,-174.2834],[299.9987,-174.2834],[300,130.0448]]
       }
      ]
     }]
   }
   {type=MovieScene3DAttachTrack, sections [{type=MovieScene3DAttachSection, end_frame=300, start_frame=0}]}
  ]
 }
]
display_rate {denominator=1, numerator=30}
tick_resolution {denominator=1, numerator=24000}
tracks [{type=MovieSceneCameraCutTrack, sections [{type=MovieSceneCameraCutSection, end_frame=300, start_frame=0}]}]
