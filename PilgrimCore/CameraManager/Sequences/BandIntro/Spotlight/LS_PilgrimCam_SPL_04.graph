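# Spotlight camera shot 04 for the PilgrimCore band intro, a Level Sequence with a single 300-frame camera cut at 30 fps.
# An attached Cine Camera makes a quick move over the first 18 frames, then drifts slowly, while focal length tightens slightly and focus and aperture stay fixed.
# A static spotLight_relocator actor accompanies it, plus an attached, animated Fort Player Mannequin whose spawn track defaults to off.
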LevelSequence /PilgrimCore/CameraManager/Sequences/BandIntro/Spotlight/LS_PilgrimCam_SPL_04
bindings [
 Cine Camera Actor {
  id=#1
  possessed_class=CineCameraActor
  tracks [
   {type=MovieSceneSpawnTrack, sections [{type=MovieSceneSpawnSection, channels [None {type=MovieSceneScriptingBoolChannel, default=true}]}]}
   {type=MovieSceneCameraShakeTrack, sections []}
   {
    type=MovieScene3DTransformTrack
    sections [{
      type=MovieScene3DTransformSection
      channels [
       Location.X {type=MovieSceneScriptingDoubleChannel, default=185.8875, keys [[0,237.912],[18,181.745],[299,156.3614]]}
       Location.Y {type=MovieSceneScriptingDoubleChannel, default=0, keys [[0,-201.4942],[18,-161.2778],[299,-143.0533]]}
       Location.Z {type=MovieSceneScriptingDoubleChannel, default=0, keys [[0,-58.3039],[18,-46.2006],[299,-40.2227]]}
       Rotation.X {type=MovieSceneScriptingDoubleChannel, default=0, const [0,3]}
       Rotation.Y {type=MovieSceneScriptingDoubleChannel, default=0, thinned_from=3, keys [[0,10.5499],[299,10.5499]]}
       Rotation.Z {type=MovieSceneScriptingDoubleChannel, default=-151.4532, keys [[0,-219.8787],[18,-222.3646],[299,-222.8166]]}
      ]
     }]
   }
   {type=MovieScene3DAttachTrack, sections [{type=MovieScene3DAttachSection, end_frame=311, start_frame=0}]}
  ]
 }
 spotLight_relocator {
  id=#2
  possessed_class=Actor
  tracks [
   {type=MovieSceneSpawnTrack, sections [{type=MovieSceneSpawnSection, channels [None {type=MovieSceneScriptingBoolChannel, default=true}]}]}
   {
    type=MovieScene3DTransformTrack
    sections [{
      type=MovieScene3DTransformSection
      channels [
       Location.X {type=MovieSceneScriptingDoubleChannel, default=0}
       Location.Y {type=MovieSceneScriptingDoubleChannel, default=0}
       Location.Z {type=MovieSceneScriptingDoubleChannel, default=0}
       Rotation.X {type=MovieSceneScriptingDoubleChannel, default=0}
       Rotation.Y {type=MovieSceneScriptingDoubleChannel, default=0}
       Rotation.Z {type=MovieSceneScriptingDoubleChannel, default=0}
       Scale.X {type=MovieSceneScriptingDoubleChannel, default=1}
       Scale.Y {type=MovieSceneScriptingDoubleChannel, default=1}
       Scale.Z {type=MovieSceneScriptingDoubleChannel, default=1}
      ]
     }]
   }
  ]
 }
 Fort Player Mannequin {
  id=#3
  possessed_class=FortPlayerMannequin
  tracks [
   {
    type=MovieScene3DTransformTrack
    sections [{
      type=MovieScene3DTransformSection
      channels [
       Location.X {type=MovieSceneScriptingDoubleChannel, default=0}
       Location.Y {type=MovieSceneScriptingDoubleChannel, default=0}
       Location.Z {type=MovieSceneScriptingDoubleChannel, default=-90.7379}
       Rotation.X {type=MovieSceneScriptingDoubleChannel, default=0}
       Rotation.Y {type=MovieSceneScriptingDoubleChannel, default=0}
       Rotation.Z {type=MovieSceneScriptingDoubleChannel, default=-90}
       Scale.X {type=MovieSceneScriptingDoubleChannel, default=1}
       Scale.Y {type=MovieSceneScriptingDoubleChannel, default=1}
       Scale.Z {type=MovieSceneScriptingDoubleChannel, default=1}
      ]
     }]
   }
   {type=MovieSceneSkeletalAnimationTrack, sections [{type=MovieSceneSkeletalAnimationSection, end_frame=354, start_frame=0}]}
   {type=MovieSceneSpawnTrack, sections [{type=MovieSceneSpawnSection, channels [None {type=MovieSceneScriptingBoolChannel, default=false}]}]}
   {type=MovieScene3DAttachTrack, sections [{type=MovieScene3DAttachSection, end_frame=354, start_frame=0}]}
  ]
 }
 CameraComponent {
  id=#4
  parent_id=#1
  possessed_class=CineCameraComponent
  tracks [
   {
    type=MovieSceneFloatTrack
    display_name=Current Focal Length
    sections [{
      type=MovieSceneFloatSection
      channels [None {type=MovieSceneScriptingFloatChannel, default=10.5197, keys [[0,14.2581],[15,15.6939],[299,16.3562]]}]
     }]
   }
   {
    type=MovieSceneFloatTrack
    display_name="Manual Focus Distance (Focus Settings)"
    sections [{type=MovieSceneFloatSection, channels [None {type=MovieSceneScriptingFloatChannel, default=222.9212}]}]
   }
   {
    type=MovieSceneFloatTrack
    display_name=Current Aperture
    sections [{type=MovieSceneFloatSection, channels [None {type=MovieSceneScriptingFloatChannel, default=0.6}]}]
   }
  ]
 }
]
display_rate {denominator=1, numerator=30}
tick_resolution {denominator=1, numerator=24000}
tracks [{type=MovieSceneCameraCutTrack, sections [{type=MovieSceneCameraCutSection, end_frame=300, start_frame=0}]}]
