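# Pilgrim camera level sequence for a spotlight shot in the band intro, with one camera cut to a spawned cine camera over 300 frames at 30 fps.
# A near-zero-scale FortPlayerCube yaws quickly at first, then slowly, apparently as an orbit pivot, alongside a static spotLight_relocator actor.
# Cube and camera both have attach tracks; lens settings are fixed and the camera shake track is empty.
# An animated player mannequin is bound but its spawn defaults to off.
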LevelSequence /PilgrimCore/CameraManager/Sequences/BandIntro/Spotlight/LS_PilgrimCam_SPL_06
bindings [
 spotLight_relocator {
  id=#1
  possessed_class=Actor
  tracks [
   {type=MovieSceneSpawnTrack, sections [{type=MovieSceneSpawnSection, channels [None {type=MovieSceneScriptingBoolChannel, default=true}]}]}
   {
    type=MovieScene3DTransformTrack
    sections [{
      type=MovieScene3DTransformSection
      channels [
       Location.X {type=MovieSceneScriptingDoubleChannel, default=0}
       Location.Y {type=MovieSceneScriptingDoubleChannel, default=0}
       Location.Z {type=MovieSceneScriptingDoubleChannel, default=0}
       Rotation.X {type=MovieSceneScriptingDoubleChannel, default=0}
       Rotation.Y {type=MovieSceneScriptingDoubleChannel, default=0}
       Rotation.Z {type=MovieSceneScriptingDoubleChannel, default=0}
       Scale.X {type=MovieSceneScriptingDoubleChannel, default=1}
       Scale.Y {type=MovieSceneScriptingDoubleChannel, default=1}
       Scale.Z {type=MovieSceneScriptingDoubleChannel, default=1}
      ]
     }]
   }
  ]
 }
 FortPlayerCube {
  id=#2
  possessed_class=FortStaticMeshActor
  tracks [
   {
    type=MovieScene3DTransformTrack
    sections [{
      type=MovieScene3DTransformSection
      channels [
       Location.X {type=MovieSceneScriptingDoubleChannel, default=0, const [0,3]}
       Location.Y {type=MovieSceneScriptingDoubleChannel, default=0, const [0,3]}
       Location.Z {type=MovieSceneScriptingDoubleChannel, default=0, const [0,3]}
       Rotation.X {type=MovieSceneScriptingDoubleChannel, default=0, const [0,3]}
       Rotation.Y {type=MovieSceneScriptingDoubleChannel, default=0, const [0,3]}
       Rotation.Z {type=MovieSceneScriptingDoubleChannel, default=-81, keys [[0,-210.2062],[18,-111.4771],[299,-55.358]]}
       Scale.X {type=MovieSceneScriptingDoubleChannel, default=0.0001, const [0.0001,1]}
       Scale.Y {type=MovieSceneScriptingDoubleChannel, default=0.0001, const [0.0001,1]}
       Scale.Z {type=MovieSceneScriptingDoubleChannel, default=0.0001, const [0.0001,1]}
      ]
     }]
   }
   {type=MovieSceneSpawnTrack, sections [{type=MovieSceneSpawnSection, channels [None {type=MovieSceneScriptingBoolChannel, default=true}]}]}
   {type=MovieScene3DAttachTrack, sections [{type=MovieScene3DAttachSection, end_frame=300, start_frame=0}]}
  ]
 }
 CameraComponent {
  id=#3
  parent_id=#4
  possessed_class=CineCameraComponent
  tracks [
   {
    type=MovieSceneFloatTrack
    display_name=Current Focal Length
    sections [{type=MovieSceneFloatSection, channels [None {type=MovieSceneScriptingFloatChannel, default=15.5431}]}]
   }
   {
    type=MovieSceneFloatTrack
    display_name="Manual Focus Distance (Focus Settings)"
    sections [{type=MovieSceneFloatSection, channels [None {type=MovieSceneScriptingFloatChannel, default=210.2038}]}]
   }
   {
    type=MovieSceneFloatTrack
    display_name=Current Aperture
    sections [{type=MovieSceneFloatSection, channels [None {type=MovieSceneScriptingFloatChannel, default=0.6}]}]
   }
  ]
 }
 Cine Camera Actor {
  id=#4
  possessed_class=CineCameraActor
  tracks [
   {type=MovieSceneSpawnTrack, sections [{type=MovieSceneSpawnSection, channels [None {type=MovieSceneScriptingBoolChannel, default=true}]}]}
   {type=MovieSceneCameraShakeTrack, sections []}
   {
    type=MovieScene3DTransformTrack
    sections [{
      type=MovieScene3DTransformSection
      channels [
       Location.X {type=MovieSceneScriptingDoubleChannel, default=185.8875, thinned_from=4, keys [[-0.0013,216.8228],[0,277599.1379],[300,277599.1379]]}
       Location.Y {type=MovieSceneScriptingDoubleChannel, default=0, thinned_from=4, keys [[-0.0013,6.2354],[0,2151287.384],[300,2151287.384]]}
       Location.Z {type=MovieSceneScriptingDoubleChannel, default=0, thinned_from=4, keys [[-0.0013,-18.1425],[0,-181424.5033],[300,-181424.5033]]}
       Rotation.X {type=MovieSceneScriptingDoubleChannel, default=0, thinned_from=4, keys [[-0.0013,0],[299.9987,0],[300,-7.6864]]}
       Rotation.Y {type=MovieSceneScriptingDoubleChannel, default=7.7811, thinned_from=4, keys [[-0.0013,7.7811],[299.9987,7.7811],[300,1.2136]]}
       Rotation.Z {type=MovieSceneScriptingDoubleChannel, default=-180, keys [[-0.0013,-180],[0,-99],[299.9987,-99],[300,179.9185]]}
      ]
     }]
   }
   {type=MovieScene3DAttachTrack, sections [{type=MovieScene3DAttachSection, end_frame=300, start_frame=0}]}
  ]
 }
 Fort Player Mannequin {
  id=#5
  possessed_class=FortPlayerMannequin
  tracks [
   {
    type=MovieScene3DTransformTrack
    sections [{
      type=MovieScene3DTransformSection
      channels [
       Location.X {type=MovieSceneScriptingDoubleChannel, default=0}
       Location.Y {type=MovieSceneScriptingDoubleChannel, default=0}
       Location.Z {type=MovieSceneScriptingDoubleChannel, default=-90.7379}
       Rotation.X {type=MovieSceneScriptingDoubleChannel, default=0}
       Rotation.Y {type=MovieSceneScriptingDoubleChannel, default=0}
       Rotation.Z {type=MovieSceneScriptingDoubleChannel, default=-90}
       Scale.X {type=MovieSceneScriptingDoubleChannel, default=1}
       Scale.Y {type=MovieSceneScriptingDoubleChannel, default=1}
       Scale.Z {type=MovieSceneScriptingDoubleChannel, default=1}
      ]
     }]
   }
   {type=MovieSceneSkeletalAnimationTrack, sections [{type=MovieSceneSkeletalAnimationSection, end_frame=354, start_frame=0}]}
   {type=MovieSceneSpawnTrack, sections [{type=MovieSceneSpawnSection, channels [None {type=MovieSceneScriptingBoolChannel, default=false}]}]}
   {type=MovieScene3DAttachTrack, sections [{type=MovieScene3DAttachSection, end_frame=354, start_frame=0}]}
  ]
 }
]
display_rate {denominator=1, numerator=30}
tick_resolution {denominator=1, numerator=24000}
tracks [{type=MovieSceneCameraCutTrack, sections [{type=MovieSceneCameraCutSection, end_frame=300, start_frame=0}]}]
