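# Band intro spotlight camera sequence 07 for PilgrimCore, a single cine camera shot of about ten seconds at 30 fps.
# The camera pulls back quickly over the opening frames, then drifts slowly outward while its focal length widens slightly and focus distance settles early.
# A spawned spotLight_relocator actor sits at identity; camera and mannequin carry attach tracks.
# The animated mannequin's spawn track defaults off, and the camera shake track is empty.
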LevelSequence /PilgrimCore/CameraManager/Sequences/BandIntro/Spotlight/LS_PilgrimCam_SPL_07
bindings [
 spotLight_relocator {
  id=#1
  possessed_class=Actor
  tracks [
   {type=MovieSceneSpawnTrack, sections [{type=MovieSceneSpawnSection, channels [None {type=MovieSceneScriptingBoolChannel, default=true}]}]}
   {
    type=MovieScene3DTransformTrack
    sections [{
      type=MovieScene3DTransformSection
      channels [
       Location.X {type=MovieSceneScriptingDoubleChannel, default=0}
       Location.Y {type=MovieSceneScriptingDoubleChannel, default=0}
       Location.Z {type=MovieSceneScriptingDoubleChannel, default=0}
       Rotation.X {type=MovieSceneScriptingDoubleChannel, default=0}
       Rotation.Y {type=MovieSceneScriptingDoubleChannel, default=0}
       Rotation.Z {type=MovieSceneScriptingDoubleChannel, default=0}
       Scale.X {type=MovieSceneScriptingDoubleChannel, default=1}
       Scale.Y {type=MovieSceneScriptingDoubleChannel, default=1}
       Scale.Z {type=MovieSceneScriptingDoubleChannel, default=1}
      ]
     }]
   }
  ]
 }
 Fort Player Mannequin {
  id=#2
  possessed_class=FortPlayerMannequin
  tracks [
   {
    type=MovieScene3DTransformTrack
    sections [{
      type=MovieScene3DTransformSection
      channels [
       Location.X {type=MovieSceneScriptingDoubleChannel, default=0}
       Location.Y {type=MovieSceneScriptingDoubleChannel, default=0}
       Location.Z {type=MovieSceneScriptingDoubleChannel, default=-90.7379}
       Rotation.X {type=MovieSceneScriptingDoubleChannel, default=0}
       Rotation.Y {type=MovieSceneScriptingDoubleChannel, default=0}
       Rotation.Z {type=MovieSceneScriptingDoubleChannel, default=-90}
       Scale.X {type=MovieSceneScriptingDoubleChannel, default=1}
       Scale.Y {type=MovieSceneScriptingDoubleChannel, default=1}
       Scale.Z {type=MovieSceneScriptingDoubleChannel, default=1}
      ]
     }]
   }
   {type=MovieSceneSkeletalAnimationTrack, sections [{type=MovieSceneSkeletalAnimationSection, end_frame=354, start_frame=0}]}
   {type=MovieSceneSpawnTrack, sections [{type=MovieSceneSpawnSection, channels [None {type=MovieSceneScriptingBoolChannel, default=false}]}]}
   {type=MovieScene3DAttachTrack, sections [{type=MovieScene3DAttachSection, end_frame=354, start_frame=0}]}
  ]
 }
 CameraComponent {
  id=#3
  parent_id=#4
  possessed_class=CineCameraComponent
  tracks [
   {
    type=MovieSceneFloatTrack
    display_name=Current Focal Length
    sections [{
      type=MovieSceneFloatSection
      channels [None {type=MovieSceneScriptingFloatChannel, default=14.7739, keys [[0,15.8555],[18,15.8555],[299,14.7739]]}]
     }]
   }
   {
    type=MovieSceneFloatTrack
    display_name="Manual Focus Distance (Focus Settings)"
    sections [{type=MovieSceneFloatSection, channels [None {type=MovieSceneScriptingFloatChannel, default=210.5191, keys [[0,140.2598],[18,210.5191]]}]}]
   }
   {
    type=MovieSceneFloatTrack
    display_name=Current Aperture
    sections [{type=MovieSceneFloatSection, channels [None {type=MovieSceneScriptingFloatChannel, default=0.6}]}]
   }
  ]
 }
 Cine Camera Actor {
  id=#4
  possessed_class=CineCameraActor
  tracks [
   {type=MovieSceneSpawnTrack, sections [{type=MovieSceneSpawnSection, channels [None {type=MovieSceneScriptingBoolChannel, default=true}]}]}
   {type=MovieSceneCameraShakeTrack, sections []}
   {
    type=MovieScene3DTransformTrack
    sections [{
      type=MovieScene3DTransformSection
      channels [
       Location.X {type=MovieSceneScriptingDoubleChannel, default=185.8875, keys [[0,115.6047],[18,193.6555],[299,214.9119]]}
       Location.Y {type=MovieSceneScriptingDoubleChannel, default=0, keys [[0,120.4176],[18,188.1603],[299,206.6093]]}
       Location.Z {type=MovieSceneScriptingDoubleChannel, default=0, keys [[0,-18.8397],[18,-30.0027],[299,-33.0429]]}
       Rotation.X {type=MovieSceneScriptingDoubleChannel, default=0, const [0,3]}
       Rotation.Y {type=MovieSceneScriptingDoubleChannel, default=0, keys [[0,10.1305],[18,6.574],[299,5.6054]]}
       Rotation.Z {type=MovieSceneScriptingDoubleChannel, default=-139.0443, thinned_from=3, keys [[0,-139.0767],[299,-139.0397]]}
      ]
     }]
   }
   {type=MovieScene3DAttachTrack, sections [{type=MovieScene3DAttachSection, end_frame=311, start_frame=0}]}
  ]
 }
]
display_rate {denominator=1, numerator=30}
tick_resolution {denominator=1, numerator=24000}
tracks [{type=MovieSceneCameraCutTrack, sections [{type=MovieSceneCameraCutSection, end_frame=300, start_frame=0}]}]
